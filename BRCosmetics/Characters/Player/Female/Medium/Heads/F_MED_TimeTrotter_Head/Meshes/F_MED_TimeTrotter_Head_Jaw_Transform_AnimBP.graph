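# Post-processing style head AnimBP for the TimeTrotter cosmetic that adds a jaw-bone rotation when the helmet variant is active.
# Each update reads the jawOpen facial curve from a part mesh's anim instance and checks the pawn's Parts variant channel for Stage2, setting Helmet.
# The cached input pose gets an additive FACIAL_C_Jaw pitch weighted by JawOpen.
# A bool blend outputs that pose when Helmet is true, otherwise the untouched input.
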AnimBlueprint /BRCosmetics/Characters/Player/Female/Medium/Heads/F_MED_TimeTrotter_Head/Meshes/F_MED_TimeTrotter_Head_Jaw_Transform_AnimBP
target_skeleton=/Game/Characters/Player/Common/Fortnite_Base_Head/Fortnite_Base_3L_Head_Skeleton
graphs [
 AnimGraph {nodes [
   n1 = "Transform (Modify) Bone - Bone: FACIAL_C_Jaw"(
    ComponentPose=Local To Component(LocalPose=Use cached pose 'Startpose'())
    Alpha=Get JawOpen() -> (JawOpen:double)
    Translation="0,0,0"
    Rotation="-15,0,0"
    Scale="1,1,1"
   ) {
    rotationMode=BMM_Additive
    translationMode=BMM_Additive
    translationSpace=BCS_BoneSpace
    alphaBoolBlend {blendOption=Linear}
    boneToModify {boneName=FACIAL_C_Jaw}
    rotation {pitch=-15}
   }
   Output Pose(Result=Blend Poses by bool(BlendPose_0=Component To Local(ComponentPose=@n1), BlendPose_1=Use cached pose 'Startpose'(), bActiveValue=Get Helmet() -> (Helmet:bool))) {layerGroup=None}
   Save cached pose 'Startpose'(Pose=Input Pose()) {cacheName=Startpose}
  ]}
 EventGraph {nodes [
   Event BlueprintUpdateAnimation()
   > n1 = Sequence()
   n2 = GetPartSkeletalMeshComponent()
   > Set JawOpen(JawOpen:double=GetCurveValue(self=GetAnimInstance(self=@n2), CurveName=CTRL_expressions_jawOpen)) -> (Output_Get:double)
   n3 = Cast To FortPlayerPawn(execute=@n1.then_1, Object=GetOwningActor())
   > n4 = GetVariantChannelActiveIndexForItemActor!(
    execute
    Item:Athena Cosmetic Item Definition Object Reference=/BRCosmetics/Athena/Items/Cosmetics/Characters/Character_TimeTrotter
    Actor:Actor Object Reference=@n3
    Channel:Gameplay Tag Structure=(TagName="Cosmetics.Variant.Channel.Parts")
    VariantList:Array of Gameplay Tag Structures=Make Array("[0]"=(TagName="Cosmetics.Variant.Property.Stage2"))
    bErrorIfChannelMissing:bool=false
    bErrorIfTagNotInArray:bool=false
   ) -> (then, Index:int=0, bIndexIsValid:bool, bFoundChannel:bool)
   > n5 = Branch(Condition=@n4.bIndexIsValid)
   > Set Helmet(Helmet:bool=true) -> (Output_Get:bool)
   Set Helmet(execute=@n5.else, Helmet=false)
  ]}
]
variables [
 Helmet {type=boolean}
 JawOpen {type=number}
 Out Names {type="array<string>", default []}
]
